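; Cheat manager extension for Rift Gauntlets that adds a RiftGauntletsSpawnMythic command to spawn the mode's mythic on demand.
; After cheat data validates, it finds the Rift Gauntlets game state component on the Athena game state.
; If both objects are valid, it delegates to that component's SpawnMythic, so the spawn logic itself lives outside this asset.
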(fn RiftGauntletsSpawnMythic
  (bind _returnvalue (Fortnite|Athena|GameState|GetGameStateAthena))
  (bind _returnvalue_2 (Actor|GetComponentbyClass _returnvalue "/RiftGauntletsGameplay/B_RiftGauntletsGameStateComponent.B_RiftGauntletsGameStateComponent_C"))
  (Default|ValidateCheatData
    (:Valid
      (Utilities|IsValid _returnvalue
        (:Is%20Valid
          (Utilities|IsValid _returnvalue_2
            (:Is%20Valid
              (Class|BRiftGauntletsGameStateComponent|SpawnMythic _returnvalue_2))))))))
