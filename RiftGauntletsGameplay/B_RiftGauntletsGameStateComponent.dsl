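; Game state component that, on the server, may spawn one Rift Gauntlets mythic pickup per match.
; Once playlist data is ready, it rolls against a hotfixable chance, then listens for the Aircraft or SafeZones game phase.
; It shuffles tagged spawner actors from the global actor registry and takes the first enabled one.
; It spawns loot tier LTG_MythicRift there, sets a once-only flag and runs the pickup's downward actor-link check.
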(event EventBeginPlay
  ;[0 0.08333 0.0132 0.5] BeginPlay: Look to spawn the mythic weapon.
  (|CallParentFunction)
  (Networking|SwitchHasAuthority
    (:Authority
      (Fortnite|FortShippingLog self "BeginPlay: Invoked")
      (Class|FortAsyncActionCurrentPlaylistReady|PlaylistDataReadyAsync
        (:OnReady
          (if (>= (ScalableFloat|GetValueatLevel (Variables|Hotfix|GetHotfix_ChanceToSpawn)) (Math|Random|RandomFloatinRange 0.01 1.0))
            (Fortnite|FortShippingLog self "BeginPlay: Spawning mythic this game.")
            (bind _returnvalue (GameplayEventRouter|GetEventRouter "/Script/FortniteGame.FortGameState" (Components|GetOwner)))
            (bind (_asynctaskproxy _payload) (GameplayEventRouter|StartListeningToStatefulEvent _returnvalue "/Script/FortniteGame.FortGamePhaseUpdatedEvent")
              (:then
                (Variables|Transient|SetAsyncTask_WaitGamePhaseChanged _asynctaskproxy))
              (:OnEventReceived
                (if (or (Utilities|Enum|Equal(Enum) (Utilities|Struct|BreakFortGamePhaseUpdatedEvent _payload) "Aircraft") (Utilities|Enum|Equal(Enum) (Utilities|Struct|BreakFortGamePhaseUpdatedEvent _payload) "SafeZones"))
                  (CallFunction|SpawnMythic)
                  (bind _asynctask_waitgamephasechanged (Variables|Transient|GetAsyncTask_WaitGamePhaseChanged)
                    (:then
                      (GameplayEventRouter|StopListeningtoStatefulEvent _asynctask_waitgamephasechanged))
                    (:else)))))
            (else
              (Fortnite|FortShippingLog self "BeginPlay: Not spawning mythic this game."))))))))

(event Custom|SpawnMythic
  ;[0.1424 0.03998 0.15 0.5] SpawnMythic: Spawn the mythic at a randomly generated location.
  (Networking|SwitchHasAuthority
    (:Authority
      (Fortnite|FortShippingLog self "SpawnMythic: Invoked")
      (if (not (Variables|Transient|GetbMythicSpawned))
        (bind _outactors (Class|FortGlobalActorRegistrySubsystem|GetRegisteredActors (WorldSubsystems|GetFortGlobalActorRegistrySubsystem) (GameplayTags|MakeLiteralGameplayTag "(TagName=\"RiftGauntlets.ActorRegister.MythicSpawner\")")))
        (bind _outitem (CallFunction|GetRandomLocationActor _outactors))
        (Utilities|IsValid _outitem
          (:Is%20Valid
            (Fortnite|FortShippingLog self (Utilities|String|Append "SpawnMythic: Spawning mythic at location " (Utilities|GetDisplayName _outitem)))
            (bind _returnvalue (Inventory|SpawnPickupInWorldWithLootTier "LTG_MythicRift" (Transformation|GetActorLocation _outitem) 0 0 false false "0.000000,0.000000,-1.000000" "FloorLoot"))
            (if (Utilities|Array|IsValidIndex _returnvalue)
              (Utilities|IsValid (Utilities|Array|Get(acopy) _returnvalue)
                (:Is%20Valid
                  (Variables|Transient|SetbMythicSpawned true)
                  (Fortnite|FortShippingLog self "SpawnMythic: Successfully spawned mythic actor." true)
                  (Pickup|RestatLocation (Utilities|Array|Get(acopy) _returnvalue) (Transformation|GetActorLocation _outitem))
                  (bind _returnvalue_2 (Actor|GetComponentbyClass (Utilities|Array|Get(acopy) _returnvalue) "/Script/FortniteGame.FortLinkToActorComponent"))
                  (Utilities|IsValid _returnvalue_2
                    (:Is%20Valid
                      (LinkToActorComponent|PerformLinkingCheck _returnvalue_2 "Down" (Variables|Transient|GetLinkedToActor)))))
                (:Is%20Not%20Valid
                  _))
              (else
                _))
            (Fortnite|FortShippingLog self "SpawnMythic: Did not spawn mythic pickup."))
          (:Is%20Not%20Valid
            (Fortnite|FortShippingLog self "SpawnMythic: No spawn locations found." true)))))))

(event EventEndPlay
  (params (EndPlayReason :type EEndPlayReason%20Enum :default "Destroyed"))
  ;[0.08333 0.05528 0 0.5] EndPlay: Clean up.
  (Utilities|FlowControl|Sequence
    (:then_0
      (Networking|SwitchHasAuthority
        (:Authority
          (bind _asynctask_waitgamephasechanged (Variables|Transient|GetAsyncTask_WaitGamePhaseChanged)
            (:then
              (GameplayEventRouter|StopListeningtoStatefulEvent _asynctask_waitgamephasechanged))
            (:else)))))
    (:then_1
      (|CallParentFunction EndPlayReason))))

(fn GetRandomLocationActor
  (params (Actors :type Array%20of%20Actor%20Object%20References))
  ;[0.15 0.05845 0.1424 0.5] Select a random spawn location actor.
  (Utilities|Array|Shuffle Actors)
  (for _array_element Actors
    (bind _asb_rift_gauntlets_mythic_spawner_location (Utilities|Casting|CastToB_RiftGauntlets_MythicSpawnerLocation _array_element))
    (if (Class|BRiftGauntletsMythicSpawnerLocation|IsLocationEnabled _asb_rift_gauntlets_mythic_spawner_location)
      (return _asb_rift_gauntlets_mythic_spawner_location)))
  (return 0))
